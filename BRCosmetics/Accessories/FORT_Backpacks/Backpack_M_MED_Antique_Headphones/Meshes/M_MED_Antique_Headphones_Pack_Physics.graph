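# Physics asset for the antique headphones backpack cosmetic, giving its dny_ bones simple collision bodies and joints.
# Capsules cover every body, with an added box on the head.
# Constraints form two unlinked trees: head and two hand chains on dny_body, and left and right line bones through dny_headphone to dny_headphone_base.
# Swing is locked at the head, both hand bases and the left hand, and all configured angle values are small.
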PhysicsAsset /BRCosmetics/Accessories/FORT_Backpacks/Backpack_M_MED_Antique_Headphones/Meshes/M_MED_Antique_Headphones_Pack_Physics
bodies {
 dny_body [{shapeType=Capsule, length=2.5342, radius=5.299, center=<-0.0675 0.5424 -6.9549>, rotation=<rot -0.649 -90.327 2.5324>}]
 dny_hand_base_l [{shapeType=Capsule, length=1.7684, radius=2.4375, rotation=<rot 0 -90 0>}]
 dny_hand_base_r [{shapeType=Capsule, length=1.7684, radius=2.4375, center=<0 0 -0.4603>, rotation=<rot 0 -90 0>}]
 dny_hand_l [{shapeType=Capsule, length=8.8862, radius=1.9667, center=<-0.2351 -1.5143 -4.88>, rotation=<rot 23.096 -90 9.1036>}]
 dny_hand_r [{shapeType=Capsule, length=8.8843, radius=1.9648, center=<0.0961 1.3583 3.9593>, rotation=<rot 23.2022 -89.9999 9.0361>}]
 dny_hand_r1 [{shapeType=Capsule, length=5.111, radius=1.8449, center=<-0.1496 2.2133 -0.2959>, rotation=<rot 0 -0.8266 -83.0924>}]
 dny_head [
  {shapeType=Capsule, length=1.841, radius=6.2234, center=<0.1569 -0.1636 1.4371>, rotation=<rot 13.307 -90 -90.2328>}
  {shapeType=Box, extentX=29.8804, extentY=5.8945, extentZ=7.5, center=<0.0001 2.3563 16.346>, rotation=<rot 0 0 -13.3691>}
 ]
 dny_headphone [
  {shapeType=Capsule, length=11.5061, radius=3.5517, center=<-0.1653 1.0487 10.8666>, rotation=<rot 5.0595 -90.1905 88.9036>}
  {shapeType=Capsule, length=1.5112, radius=5.2793}
 ]
 dny_headphone_base [{shapeType=Capsule, length=0, radius=7.2929, center=<0 0 -0.4751>, rotation=<rot 0 -90 0>}]
 dny_headphone_line01_l [{shapeType=Capsule, length=4.9926, radius=2.0438, center=<2.2357 -0.8959 0.2374>, rotation=<rot -10.9062 94.4715 -95.3064>}]
 dny_headphone_rine01_r [{shapeType=Capsule, length=4.706, radius=2.1638, center=<-2.1919 0.4484 -0.2895>, rotation=<rot -11.8096 81.417 -80.0512>}]
}
constraints [
 {
  bone1Name=dny_hand_l
  bone2Name=dny_hand_base_l
  swing1LimitDegrees=1
  swing1Motion=Locked
  swing2LimitDegrees=1
  swing2Motion=Locked
  twistLimitDegrees=15
 }
 {bone1Name=dny_hand_r, bone2Name=dny_hand_base_r, swing1LimitDegrees=1, swing2LimitDegrees=1, twistLimitDegrees=10}
 {bone1Name=dny_hand_r1, bone2Name=dny_hand_r, swing1LimitDegrees=1, swing2LimitDegrees=1, twistLimitDegrees=10}
 {bone1Name=dny_headphone_rine01_r, bone2Name=dny_headphone, swing1LimitDegrees=15, swing2LimitDegrees=0.1, twistLimitDegrees=0.1}
 {bone1Name=dny_headphone_line01_l, bone2Name=dny_headphone, swing1LimitDegrees=15, swing2LimitDegrees=0.1, twistLimitDegrees=0.1}
 {
  bone1Name=dny_head
  bone2Name=dny_body
  swing1LimitDegrees=0.01
  swing1Motion=Locked
  swing2LimitDegrees=0.01
  swing2Motion=Locked
  twistLimitDegrees=0.01
 }
 {
  bone1Name=dny_hand_base_r
  bone2Name=dny_body
  swing1LimitDegrees=1
  swing1Motion=Locked
  swing2LimitDegrees=1
  swing2Motion=Locked
  twistLimitDegrees=10
 }
 {
  bone1Name=dny_hand_base_l
  bone2Name=dny_body
  swing1LimitDegrees=1
  swing1Motion=Locked
  swing2LimitDegrees=1
  swing2Motion=Locked
  twistLimitDegrees=10
 }
 {bone1Name=dny_headphone, bone2Name=dny_headphone_base, swing1LimitDegrees=0.1, swing2LimitDegrees=0.1, twistLimitDegrees=2}
]
